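# Physics asset for the front skirt straps of the male medium Citadel armor, giving them jointed physical motion.
# Left and right chains of three capsules hang from tiny FrontStrap anchor bodies, with tight swing limits, zero twist, and wider swing on one axis for the lower two joints.
# Pelvis and thigh capsules supply leg and hip shapes, with the thighs and strap anchors joined to the pelvis by angularly free constraints.
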PhysicsAsset /BRCosmetics/Characters/Player/Male/Medium/Bodies/M_MED_Citadel_Armor/Meshes/M_MED_Citadel_Armor_FrontSkirt_Physics1
bodies {
 FrontStrap1_dyn_l [{shapeType=Capsule, length=12.75, radius=6.75, shapeName=FrontStrap1_dyn_l_capsule, center=<10 0 4.5>, rotation=<rot 90 0 0>}]
 FrontStrap1_dyn_r [{shapeType=Capsule, length=12.75, radius=6.75, shapeName=FrontStrap1_dyn_r_capsule, center=<-10 0 -4.5>, rotation=<rot 90 0 0>}]
 FrontStrap2_dyn_l [{shapeType=Capsule, length=12.75, radius=6.75, shapeName=FrontStrap2_dyn_l_capsule, center=<10 0 4.5>, rotation=<rot 90 0 0>}]
 FrontStrap2_dyn_r [{shapeType=Capsule, length=12.75, radius=6.75, shapeName=FrontStrap2_dyn_r_capsule, center=<-10 0 -4.5>, rotation=<rot 90 0 0>}]
 FrontStrap3_dyn_l [{shapeType=Capsule, length=12.75, radius=6.75, shapeName=FrontStrap3_dyn_l_capsule, center=<10 0 4.5>, rotation=<rot 90 0 0>}]
 FrontStrap3_dyn_r [{shapeType=Capsule, length=12.75, radius=6.75, shapeName=FrontStrap3_dyn_r_capsule, center=<-10 0 -4.5>, rotation=<rot 90 0 0>}]
 FrontStrap_l [{
   shapeType=Capsule
   length=0.1
   radius=0.1
   shapeName=FrontStrap_l_capsule
   center=<4.0277 -0.5716 0.3611>
   rotation=<rot 0 -169.3023 -82.8814>
  }]
 FrontStrap_r [{
   shapeType=Capsule
   length=0.1
   radius=0.1
   shapeName=FrontStrap_r_capsule
   center=<-4.0277 0.5716 -0.3611>
   rotation=<rot 0 -169.3023 -82.8814>
  }]
 pelvis [{shapeType=Capsule, length=25.1366, radius=7.4229, shapeName=pelvis_capsule, center=<1.32 1.4025 0.0085>, rotation=<rot 0.3318 82.4165 -90>}]
 thigh_l [{shapeType=Capsule, length=37.2111, radius=6.8633, shapeName=thigh_l_capsule, center=<-17.9452 0 0>, rotation=<rot 90 0 0>}]
 thigh_r [{shapeType=Capsule, length=36.4611, radius=7.6133, shapeName=thigh_r_capsule, center=<17.9452 0 0>, rotation=<rot 90 0 0>}]
}
constraints [
 {bone1Name=FrontStrap1_dyn_l, bone2Name=FrontStrap_l, swing1LimitDegrees=10, swing2LimitDegrees=10, twistLimitDegrees=0}
 {bone1Name=FrontStrap2_dyn_l, bone2Name=FrontStrap1_dyn_l, swing1LimitDegrees=10, swing2LimitDegrees=30, twistLimitDegrees=0}
 {bone1Name=FrontStrap3_dyn_l, bone2Name=FrontStrap2_dyn_l, swing1LimitDegrees=10, swing2LimitDegrees=30, twistLimitDegrees=0}
 {bone1Name=FrontStrap1_dyn_r, bone2Name=FrontStrap_r, swing1LimitDegrees=10, swing2LimitDegrees=10, twistLimitDegrees=0}
 {bone1Name=FrontStrap2_dyn_r, bone2Name=FrontStrap1_dyn_r, swing1LimitDegrees=10, swing2LimitDegrees=30, twistLimitDegrees=0}
 {bone1Name=FrontStrap3_dyn_r, bone2Name=FrontStrap2_dyn_r, swing1LimitDegrees=10, swing2LimitDegrees=30, twistLimitDegrees=0}
 {
  bone1Name=thigh_l
  bone2Name=pelvis
  swing1LimitDegrees=45
  swing1Motion=Free
  swing2LimitDegrees=45
  swing2Motion=Free
  twistLimitDegrees=45
  twistMotion=Free
 }
 {
  bone1Name=thigh_r
  bone2Name=pelvis
  swing1LimitDegrees=45
  swing1Motion=Free
  swing2LimitDegrees=45
  swing2Motion=Free
  twistLimitDegrees=45
  twistMotion=Free
 }
 {
  bone1Name=FrontStrap_l
  bone2Name=pelvis
  swing1LimitDegrees=45
  swing1Motion=Free
  swing2LimitDegrees=45
  swing2Motion=Free
  twistLimitDegrees=45
  twistMotion=Free
 }
 {
  bone1Name=FrontStrap_r
  bone2Name=pelvis
  swing1LimitDegrees=45
  swing1Motion=Free
  swing2LimitDegrees=45
  swing2Motion=Free
  twistLimitDegrees=45
  twistMotion=Free
 }
]
