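# Physics asset for the male medium Knight Boss NPC body, with capsule bodies for a dynamic cape and tail plus the pelvis and spine.
# Center, left and right cape chains hang from spine_05 with tight twist and asymmetric swing limits, except the free-rotating center master joint.
# Cross-constraints at the fourth and seventh segments tie the side chains to the center.
# The four-segment tail hangs from the pelvis with free rotation, as do the spine joints.
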PhysicsAsset /BRCosmetics/Characters/Player/Male/Medium/Bodies/M_MED_Knight_Boss_NPC/Meshes/M_MED_Knight_Boss_NPC_Physics
bodies {
 Tail1_dyn [{shapeType=Capsule, length=10, radius=6, shapeName=Tail1_dyn_capsule, center=<0 0 -3.8776>, rotation=<rot 0 0 90>}]
 Tail2_dyn [{shapeType=Capsule, length=10, radius=6, shapeName=Tail2_dyn_capsule, center=<0 0 -3.8776>, rotation=<rot 0 0 90>}]
 Tail3_dyn [{shapeType=Capsule, length=10, radius=6, shapeName=Tail3_dyn_capsule, center=<0 0 -3.8776>, rotation=<rot 0 0 90>}]
 Tail4_dyn [{shapeType=Capsule, length=10, radius=6, shapeName=Tail4_dyn_capsule, center=<0 0 -3.8776>, rotation=<rot 0 0 90>}]
 dyn_cape_01 [{shapeType=Capsule, length=1, radius=8, shapeName=dyn_cape_01_capsule, center=<10 0 0>}]
 dyn_cape_01_l [{shapeType=Capsule, length=0.505, radius=8, shapeName=dyn_cape_01_l_capsule, center=<10 0 0>}]
 dyn_cape_01_r [{shapeType=Capsule, length=0.505, radius=8, shapeName=dyn_cape_01_r_capsule, center=<10 0 0>}]
 dyn_cape_02 [{shapeType=Capsule, length=3, radius=8, shapeName=dyn_cape_02_capsule, center=<5 0 2>, rotation=<rot 90 0 0>}]
 dyn_cape_02_l [{shapeType=Capsule, length=3, radius=8, shapeName=dyn_cape_02_l_capsule, center=<5 0 2>, rotation=<rot 90 0 0>}]
 dyn_cape_02_r [{shapeType=Capsule, length=3, radius=8, shapeName=dyn_cape_02_r_capsule, center=<5 0 2>, rotation=<rot 90 0 0>}]
 dyn_cape_03 [{shapeType=Capsule, length=10, radius=8, shapeName=dyn_cape_03_capsule, center=<10 0 3>, rotation=<rot 90 0 0>}]
 dyn_cape_03_l [{shapeType=Capsule, length=10, radius=8, shapeName=dyn_cape_03_l_capsule, center=<10 0 3>, rotation=<rot 90 0 0>}]
 dyn_cape_03_r [{shapeType=Capsule, length=10, radius=8, shapeName=dyn_cape_03_r_capsule, center=<10 0 3>, rotation=<rot 90 0 0>}]
 dyn_cape_04 [{shapeType=Capsule, length=6, radius=8, shapeName=dyn_cape_04_capsule, center=<10 0 2>, rotation=<rot 90 0 0>}]
 dyn_cape_04_l [{shapeType=Capsule, length=6, radius=8, shapeName=dyn_cape_04_l_capsule, center=<10 0 2>, rotation=<rot 90 0 0>}]
 dyn_cape_04_r [{shapeType=Capsule, length=6, radius=8, shapeName=dyn_cape_04_r_capsule, center=<10 0 2>, rotation=<rot 90 0 0>}]
 dyn_cape_05 [{shapeType=Capsule, length=6, radius=8, shapeName=dyn_cape_05_capsule, center=<10 0 2>, rotation=<rot 90 0 0>}]
 dyn_cape_05_l [{shapeType=Capsule, length=6, radius=8, shapeName=dyn_cape_05_l_capsule, center=<10 0 2>, rotation=<rot 90 0 0>}]
 dyn_cape_05_r [{shapeType=Capsule, length=6, radius=8, shapeName=dyn_cape_05_r_capsule, center=<10 0 2>, rotation=<rot 90 0 0>}]
 dyn_cape_06 [{shapeType=Capsule, length=2, radius=8, shapeName=dyn_cape_06_capsule, center=<5 0 2>, rotation=<rot 90 0 0>}]
 dyn_cape_06_l [{shapeType=Capsule, length=2, radius=8, shapeName=dyn_cape_06_l_capsule, center=<5 0 2>, rotation=<rot 90 0 0>}]
 dyn_cape_06_r [{shapeType=Capsule, length=2, radius=8, shapeName=dyn_cape_06_r_capsule, center=<5 0 2>, rotation=<rot 90 0 0>}]
 dyn_cape_07 [{shapeType=Capsule, length=1, radius=8, shapeName=dyn_cape_07_capsule, center=<5 0 2>, rotation=<rot 90 0 0>}]
 dyn_cape_07_l [{shapeType=Capsule, length=1, radius=8, shapeName=dyn_cape_07_l_capsule, center=<5 0 2>, rotation=<rot 90 0 0>}]
 dyn_cape_07_r [{shapeType=Capsule, length=1, radius=8, shapeName=dyn_cape_07_r_capsule, center=<5 0 2>, rotation=<rot 90 0 0>}]
 dyn_cape_MSTR [{shapeType=Capsule, length=1, radius=1, shapeName=dyn_cape_MSTR_capsule, rotation=<rot 0 -90 0>}]
 dyn_cape_mstr_l [{shapeType=Capsule, length=0.505, radius=8, shapeName=dyn_cape_mstr_l_capsule, center=<10 0 0>}]
 dyn_cape_mstr_r [{shapeType=Capsule, length=0.505, radius=8, shapeName=dyn_cape_mstr_r_capsule, center=<10 0 0>}]
 pelvis [{shapeType=Capsule, length=9, radius=12.5, shapeName=pelvis_capsule}]
 spine_01 [{shapeType=Capsule, length=9, radius=12.5, shapeName=spine_01_capsule}]
 spine_02 [{shapeType=Capsule, length=7.5, radius=11, shapeName=spine_02_capsule}]
 spine_03 [{shapeType=Capsule, length=9, radius=10.5, shapeName=spine_03_capsule}]
 spine_04 [{shapeType=Capsule, length=7, radius=12.5, shapeName=spine_04_capsule}]
 spine_05 [{shapeType=Capsule, length=9, radius=12.5, shapeName=spine_05_capsule}]
}
constraints [
 {bone1Name=dyn_cape_01, bone2Name=dyn_cape_MSTR, swing1LimitDegrees=20, swing2LimitDegrees=60, twistLimitDegrees=1}
 {bone1Name=dyn_cape_02, bone2Name=dyn_cape_01, swing1LimitDegrees=10, swing2LimitDegrees=60, twistLimitDegrees=1}
 {bone1Name=dyn_cape_03, bone2Name=dyn_cape_02, swing1LimitDegrees=10, swing2LimitDegrees=60, twistLimitDegrees=1}
 {bone1Name=dyn_cape_04, bone2Name=dyn_cape_03, swing1LimitDegrees=10, swing2LimitDegrees=60, twistLimitDegrees=1}
 {bone1Name=dyn_cape_05, bone2Name=dyn_cape_04, swing1LimitDegrees=10, swing2LimitDegrees=60, twistLimitDegrees=1}
 {bone1Name=dyn_cape_06, bone2Name=dyn_cape_05, swing1LimitDegrees=10, swing2LimitDegrees=60, twistLimitDegrees=1}
 {bone1Name=dyn_cape_07, bone2Name=dyn_cape_06, swing1LimitDegrees=10, swing2LimitDegrees=60, twistLimitDegrees=1}
 {
  bone1Name=dyn_cape_MSTR
  bone2Name=spine_05
  swing1LimitDegrees=45
  swing1Motion=Free
  swing2LimitDegrees=45
  swing2Motion=Free
  twistLimitDegrees=45
  twistMotion=Free
 }
 {
  bone1Name=Tail1_dyn
  bone2Name=pelvis
  swing1LimitDegrees=45
  swing1Motion=Free
  swing2LimitDegrees=45
  swing2Motion=Free
  twistLimitDegrees=45
  twistMotion=Free
 }
 {
  bone1Name=Tail2_dyn
  bone2Name=Tail1_dyn
  swing1LimitDegrees=45
  swing1Motion=Free
  swing2LimitDegrees=45
  swing2Motion=Free
  twistLimitDegrees=45
  twistMotion=Free
 }
 {
  bone1Name=Tail3_dyn
  bone2Name=Tail2_dyn
  swing1LimitDegrees=45
  swing1Motion=Free
  swing2LimitDegrees=45
  swing2Motion=Free
  twistLimitDegrees=45
  twistMotion=Free
 }
 {
  bone1Name=Tail4_dyn
  bone2Name=Tail3_dyn
  swing1LimitDegrees=45
  swing1Motion=Free
  swing2LimitDegrees=45
  swing2Motion=Free
  twistLimitDegrees=45
  twistMotion=Free
 }
 {
  bone1Name=spine_01
  bone2Name=pelvis
  swing1LimitDegrees=45
  swing1Motion=Free
  swing2LimitDegrees=45
  swing2Motion=Free
  twistLimitDegrees=45
  twistMotion=Free
 }
 {
  bone1Name=spine_02
  bone2Name=spine_01
  swing1LimitDegrees=45
  swing1Motion=Free
  swing2LimitDegrees=45
  swing2Motion=Free
  twistLimitDegrees=45
  twistMotion=Free
 }
 {
  bone1Name=spine_03
  bone2Name=spine_02
  swing1LimitDegrees=45
  swing1Motion=Free
  swing2LimitDegrees=45
  swing2Motion=Free
  twistLimitDegrees=45
  twistMotion=Free
 }
 {
  bone1Name=spine_04
  bone2Name=spine_03
  swing1LimitDegrees=45
  swing1Motion=Free
  swing2LimitDegrees=45
  swing2Motion=Free
  twistLimitDegrees=45
  twistMotion=Free
 }
 {
  bone1Name=spine_05
  bone2Name=spine_04
  swing1LimitDegrees=45
  swing1Motion=Free
  swing2LimitDegrees=45
  swing2Motion=Free
  twistLimitDegrees=45
  twistMotion=Free
 }
 {bone1Name=dyn_cape_mstr_l, bone2Name=spine_05, swing1LimitDegrees=20, swing2LimitDegrees=60, twistLimitDegrees=1}
 {bone1Name=dyn_cape_01_l, bone2Name=dyn_cape_mstr_l, swing1LimitDegrees=20, swing2LimitDegrees=60, twistLimitDegrees=1}
 {bone1Name=dyn_cape_02_l, bone2Name=dyn_cape_01_l, swing1LimitDegrees=10, swing2LimitDegrees=60, twistLimitDegrees=1}
 {bone1Name=dyn_cape_03_l, bone2Name=dyn_cape_02_l, swing1LimitDegrees=10, swing2LimitDegrees=60, twistLimitDegrees=1}
 {bone1Name=dyn_cape_04_l, bone2Name=dyn_cape_03_l, swing1LimitDegrees=10, swing2LimitDegrees=60, twistLimitDegrees=1}
 {bone1Name=dyn_cape_05_l, bone2Name=dyn_cape_04_l, swing1LimitDegrees=10, swing2LimitDegrees=60, twistLimitDegrees=1}
 {bone1Name=dyn_cape_06_l, bone2Name=dyn_cape_05_l, swing1LimitDegrees=10, swing2LimitDegrees=60, twistLimitDegrees=1}
 {bone1Name=dyn_cape_07_l, bone2Name=dyn_cape_06_l, swing1LimitDegrees=10, swing2LimitDegrees=60, twistLimitDegrees=1}
 {bone1Name=dyn_cape_mstr_r, bone2Name=spine_05, swing1LimitDegrees=20, swing2LimitDegrees=60, twistLimitDegrees=1}
 {bone1Name=dyn_cape_01_r, bone2Name=dyn_cape_mstr_r, swing1LimitDegrees=20, swing2LimitDegrees=60, twistLimitDegrees=1}
 {bone1Name=dyn_cape_02_r, bone2Name=dyn_cape_01_r, swing1LimitDegrees=10, swing2LimitDegrees=60, twistLimitDegrees=1}
 {bone1Name=dyn_cape_03_r, bone2Name=dyn_cape_02_r, swing1LimitDegrees=10, swing2LimitDegrees=60, twistLimitDegrees=1}
 {bone1Name=dyn_cape_04_r, bone2Name=dyn_cape_03_r, swing1LimitDegrees=10, swing2LimitDegrees=60, twistLimitDegrees=1}
 {bone1Name=dyn_cape_05_r, bone2Name=dyn_cape_04_r, swing1LimitDegrees=10, swing2LimitDegrees=60, twistLimitDegrees=1}
 {bone1Name=dyn_cape_06_r, bone2Name=dyn_cape_05_r, swing1LimitDegrees=10, swing2LimitDegrees=60, twistLimitDegrees=1}
 {bone1Name=dyn_cape_07_r, bone2Name=dyn_cape_06_r, swing1LimitDegrees=10, swing2LimitDegrees=60, twistLimitDegrees=1}
 {bone1Name=dyn_cape_07_r, bone2Name=dyn_cape_07, swing1LimitDegrees=45, swing2LimitDegrees=45, twistLimitDegrees=45}
 {bone1Name=dyn_cape_07_l, bone2Name=dyn_cape_07, swing1LimitDegrees=45, swing2LimitDegrees=45, twistLimitDegrees=45}
 {bone1Name=dyn_cape_04, bone2Name=dyn_cape_04_l, swing1LimitDegrees=45, swing2LimitDegrees=45, twistLimitDegrees=45}
 {bone1Name=dyn_cape_04_r, bone2Name=dyn_cape_04, swing1LimitDegrees=45, swing2LimitDegrees=45, twistLimitDegrees=45}
]
