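# Animation Blueprint for the male medium Knight Boss NPC cosmetic body, following the attached parent's pose with secondary motion.
# The output chain copies that pose, runs three RigidBody passes (two borrowed Citadel Armor skirt assets, one its own), then offsets hip pad, cape and strap bones.
# The event graph computes gravity override, crouch, pistol, arm and glider values.
# Shoulder-pad, Control Rig and chest constraint nodes do not reach the output.
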AnimBlueprint /BRCosmetics/Characters/Player/Male/Medium/Bodies/M_MED_Knight_Boss_NPC/Meshes/M_MED_Knight_Boss_NPC_AnimBP
target_skeleton=&/M_MED_Knight_Boss_NPC_Skeleton
graphs [
 AnimGraph {nodes [
   "Constraint - chest (PPPPPPPP)"!(
    ComponentPose
    Alpha=1
    ConstraintWeights_0=0.0302
    ConstraintWeights_1:float=0.1422
    ConstraintWeights_2:float=0.2966
    ConstraintWeights_3:float=0.339
    ConstraintWeights_4:float=0.1082
    ConstraintWeights_5:float=0.0032
    ConstraintWeights_6:float=0.0403
    ConstraintWeights_7:float=0.0403
   ) -> (Pose) {
    alphaBoolBlend {blendOption=Linear}
    boneToModify {boneName=chest}
    constraintSetup [
     {targetBone {boneName=spine_01}}
     {targetBone {boneName=spine_02}}
     {targetBone {boneName=spine_03}}
     {targetBone {boneName=spine_04}}
     {targetBone {boneName=spine_05}}
     {targetBone {boneName=clavicle_r}}
     {targetBone {boneName=deform_pec_r}}
     {targetBone {boneName=deform_pec_l}}
    ]
    constraintWeights [1,1,1,1,1,1,1,1]
   }
   n1 = Control Rig() {
    alphaBoolBlend {blendOption=Linear}
    controlRigAssetReference {blueprintRigClass=/BRCosmetics/Characters/Player/Male/Medium/Bodies/M_MED_Citadel_Armor/Meshes/M_MED_DeformRig_CitadelArmor_ControlRig_new1.M_MED_DeformRig_CitadelArmor_ControlRig_new1_C}
   }
   Control Rig!(Source=@n1, bAlphaBoolEnabled:bool=NOT Boolean(A=Get bIsSkyDiving())) -> (Pose) {
    alphaInputType=Bool
    alphaBoolBlend {blendOption=Linear}
    controlRigAssetReference {blueprintRigClass=/BRCosmetics/Characters/Player/Male/Medium/Bodies/M_MED_Citadel_Armor/Meshes/M_MED_Citadel_Armor_CtrlRig.M_MED_Citadel_Armor_CtrlRig_C}
   }
   Copy Pose From Mesh() {bUseAttachedParent=true}
   n2 = Get bIsSkydiveFloating()
   n3 = "Transform (Modify) Bone - Bone: upperarm_shldrPad_l"!(ComponentPose, bAlphaBoolEnabled:bool=@n2, Translation="0,0,3", Rotation="0,0,0", Scale="0.9,0.9,0.9") -> (Pose) {
    alphaInputType=Bool
    rotationMode=BMM_Additive
    scaleMode=BMM_Additive
    translationMode=BMM_Additive
    alphaBoolBlend {blendOption=Linear}
    alphaScaleBiasClamp {bInterpResult=true}
    boneToModify {boneName=upperarm_shldrPad_l}
    scale=<0.9 0.9 0.9>
    translation {z=3}
   }
   n4 = "Transform (Modify) Bone - Bone: upperarm_shldrPad_r"!(ComponentPose=@n3, bAlphaBoolEnabled:bool=@n2, Translation="0,-1,3", Rotation="0,0,0", Scale="0.9,0.9,0.9") -> (Pose) {
    alphaInputType=Bool
    rotationMode=BMM_Additive
    scaleMode=BMM_Additive
    translationMode=BMM_Additive
    alphaBoolBlend {blendOption=Linear}
    alphaScaleBiasClamp {bInterpResult=true}
    boneToModify {boneName=upperarm_shldrPad_r}
    scale=<0.9 0.9 0.9>
    translation {y=-1, z=3}
   }
   n5 = Get IsHangGlider() -> (IsHangGlider:double)
   n6 = "Transform (Modify) Bone - Bone: upperarm_shldrPad_l"!(ComponentPose=@n4, Alpha:float=@n5, Translation="6,-3,5", Rotation="0,0,0", Scale="1,1,1") -> (Pose) {
    rotationMode=BMM_Additive
    scaleMode=BMM_Additive
    translationMode=BMM_Additive
    alphaBoolBlend {blendOption=Linear}
    alphaScaleBiasClamp {bInterpResult=true}
    boneToModify {boneName=upperarm_shldrPad_l}
    translation=<6 -3 5>
   }
   n7 = "Transform (Modify) Bone - Bone: upperarm_shldrPad_r"!(ComponentPose=@n6, Alpha:float=@n5, Translation="-6,-3,5", Rotation="0,0,0", Scale="1,1,1") -> (Pose) {
    rotationMode=BMM_Additive
    scaleMode=BMM_Additive
    translationMode=BMM_Additive
    alphaBoolBlend {blendOption=Linear}
    alphaScaleBiasClamp {bInterpResult=true}
    boneToModify {boneName=upperarm_shldrPad_r}
    translation=<-6 -3 5>
   }
   n8 = "Transform (Modify) Bone - Bone: upperarm_shldrPad_r"(ComponentPose=@n7, Alpha=Get ArmRtDn() -> (ArmRtDn:float), Translation="0,0,0", Rotation="0,20,0", Scale="1,1,1") {
    rotationMode=BMM_Additive
    scaleMode=BMM_Additive
    translationMode=BMM_Additive
    alphaBoolBlend {blendOption=Linear}
    alphaScaleBiasClamp {bInterpResult=true, interpSpeedDecreasing=1, interpSpeedIncreasing=20}
    boneToModify {boneName=upperarm_shldrPad_r}
    rotation {yaw=20}
   }
   "Transform (Modify) Bone - Bone: upperarm_shldrPad_l"(ComponentPose=@n8, Alpha=Get ArmLfDn() -> (ArmLfDn:double), Translation="0,2,0", Rotation="0,-15,0", Scale="1,1,1") {
    rotationMode=BMM_Additive
    scaleMode=BMM_Additive
    translationMode=BMM_Additive
    alphaBoolBlend {blendOption=Linear}
    alphaScaleBiasClamp {bInterpResult=true, interpSpeedDecreasing=5, interpSpeedIncreasing=5}
    boneToModify {boneName=upperarm_shldrPad_l}
    rotation {yaw=-15}
    translation {y=2}
   }
   n9 = Get GravityOverride()
   n10 = RigidBody!(
    ComponentPose=Local To Component(LocalPose=Copy Pose From Mesh() {bUseAttachedParent=true})
    Alpha
    OverrideWorldGravity:Vector=@n9
    ExternalForce
   ) -> (Pose) {
    bClampLinearTranslationLimitToRefPose=true
    bOverrideWorldGravity=true
    overridePhysicsAsset=/BRCosmetics/Characters/Player/Male/Medium/Bodies/M_MED_Citadel_Armor/Meshes/M_MED_Citadel_Armor_BackSkirt_Physics
    simulationSpace=BaseBoneSpace
    alphaBoolBlend {blendOption=Linear}
    baseBoneRef {boneName=root}
    componentAppliedLinearAccClamp=<5000 5000 5000>
    componentLinearAccScale {z=0.6}
    componentLinearVelScale {z=0.6}
    overrideWorldGravity {z=-980}
    simSpaceSettings {
     maxAngularAcceleration=500
     maxAngularVelocity=3
     maxLinearAcceleration=700
     maxLinearVelocity=500
     velocityScaleZ=0
     worldAlpha=0.25
     externalLinearDragV=<1 1>
    }
   }
   n11 = RigidBody(ComponentPose=@n10, OverrideWorldGravity=@n9) {
    bClampLinearTranslationLimitToRefPose=true
    bOverrideWorldGravity=true
    overridePhysicsAsset=/BRCosmetics/Characters/Player/Male/Medium/Bodies/M_MED_Citadel_Armor/Meshes/M_MED_Citadel_Armor_FrontSkirt_Physics1
    simulationSpace=BaseBoneSpace
    alphaBoolBlend {blendOption=Linear}
    baseBoneRef {boneName=root}
    componentAppliedLinearAccClamp=<5000 5000 5000>
    componentLinearAccScale {z=0.6}
    componentLinearVelScale {z=0.6}
    overrideWorldGravity {z=-980}
    simSpaceSettings {
     maxAngularAcceleration=500
     maxAngularVelocity=3
     maxLinearAcceleration=700
     maxLinearVelocity=500
     velocityScaleZ=0
     worldAlpha=0.25
     externalLinearDragV=<1 1>
    }
   }
   n12 = RigidBody!(ComponentPose=@n11, Alpha, ExternalForce) -> (Pose) {
    bClampLinearTranslationLimitToRefPose=true
    bOverrideWorldGravity=true
    overridePhysicsAsset=&/M_MED_Knight_Boss_NPC_Physics
    simulationSpace=BaseBoneSpace
    alphaBoolBlend {blendOption=Linear}
    baseBoneRef {boneName=root}
    componentLinearAccScale=<1 1 0.6>
    componentLinearVelScale=<1 1 0.6>
    overrideWorldGravity {z=-980}
    simSpaceSettings {
     maxAngularAcceleration=500
     maxAngularVelocity=3
     maxLinearAcceleration=700
     maxLinearVelocity=500
     velocityScaleZ=0
     worldAlpha=0.25
     externalLinearDragV=<1 1>
    }
   }
   n13 = "Transform (Modify) Bone - Bone: HipPad_r"!(ComponentPose=@n12, bAlphaBoolEnabled:bool=Get bIsCrouching(), Translation="0,0,0", Rotation="-20,0,0", Scale="1,1,1") -> (Pose) {
    alphaInputType=Bool
    rotationMode=BMM_Additive
    rotationSpace=BCS_BoneSpace
    alphaBoolBlend {blendOption=Linear}
    alphaScaleBiasClamp {bInterpResult=true, interpSpeedDecreasing=0.1, interpSpeedIncreasing=0.1}
    boneToModify {boneName=HipPad_r}
    rotation {pitch=-20}
   }
   n14 = "Transform (Modify) Bone - Bone: cape_top"(ComponentPose=@n13, Alpha=1, Translation="0,-3,0", Rotation="0,0,0", Scale="1,1,1") {
    rotationMode=BMM_Additive
    rotationSpace=BCS_BoneSpace
    translationMode=BMM_Additive
    translationSpace=BCS_BoneSpace
    alphaBoolBlend {blendOption=Linear}
    alphaScaleBiasClamp {bInterpResult=true, interpSpeedDecreasing=0.1, interpSpeedIncreasing=0.1}
    boneToModify {boneName=cape_top}
    translation {y=-3}
   }
   n15 = Get Is Crouching NotMoving Float() -> (Is Crouching NotMoving Float:double)
   n16 = "Transform (Modify) Bone - Bone: HipPadMain_r"(ComponentPose=@n14, Alpha=@n15, Translation="0,0,0", Rotation="-10,0,0", Scale="1,1,1") {
    rotationMode=BMM_Additive
    alphaBoolBlend {blendOption=Linear}
    alphaScaleBiasClamp {bInterpResult=true, interpSpeedDecreasing=0.1, interpSpeedIncreasing=0.1}
    boneToModify {boneName=HipPadMain_r}
    rotation {pitch=-10}
   }
   n17 = "Transform (Modify) Bone - Bone: FrontStrap1_dyn_l"(ComponentPose=@n16, Alpha=@n15, Translation="0,0,1", Rotation="0,0,20", Scale="1,1,1") {
    rotationMode=BMM_Additive
    translationMode=BMM_Additive
    alphaBoolBlend {blendOption=Linear}
    alphaScaleBiasClamp {bInterpResult=true}
    boneToModify {boneName=FrontStrap1_dyn_l}
    rotation {roll=20}
    translation {z=1}
   }
   Output Pose(Result=Component To Local(ComponentPose=@n17)) {layerGroup=None}
  ]}
 EventGraph {nodes [
   Event BlueprintUpdateAnimation()
   > n1 = Sequence!(execute) -> (then_0, then_1, then_2:Exec, then_3:Exec, then_4:Exec)
   > n2 = Gravity Override State Function DEPRECATED(CharacterPartAnimInstance=Self-Reference(), GameTimeInSeconds=GetGameTimeInSeconds(), GravityOverrideParams=Get Gravity Override Params())
   > Set GravityOverride(GravityOverride=@n2)
   n3 = NOT Boolean(A=Get bIsMoving2D())
   n4 = Get bIsCrouching()
   Set Is Crouching NotMoving Float(execute=@n1.then_1, Is Crouching NotMoving Float:double="To Float (Boolean)"(InBool=AND Boolean(A=@n4, B=@n3))) -> (Output_Get:double)
   n5 = Cast To FortPlayerAnimInstance(execute=@n1.then_2, Object=GetAnimInstance(self=GetOwnerSkeletalMeshComponent()))
   > Set ar_firingPosition(ar_firingPosition:double=0) -> (Output_Get:double)
   > n6 = Switch on EFortWeaponCoreAnimation!(execute, Selection:EFortWeaponCoreAnimation Enum=Get WeaponCoreAnim(self=@n5)) -> (
    Melee:Exec
    Pistol:Exec
    Shotgun:Exec
    PaperBlueprint:Exec
    Rifle:Exec
    MeleeOneHand:Exec
    MachinePistol:Exec
    RocketLauncher:Exec
    GrenadeLauncher:Exec
    GoingCommando:Exec
    AssaultRifle:Exec
    TacticalShotgun:Exec
    SniperRifle:Exec
    TrapPlacement:Exec
    ShoulderLauncher:Exec
    AbilityDecoTool:Exec
    Crossbow:Exec
    C4:Exec
    RemoteControl:Exec
    DualWield:Exec
    AR_BullPup:Exec
    AR_ForwardGrip:Exec
    MedPackPaddles:Exec
    SMG_P90:Exec
    AR_DrumGun:Exec
    Consumable_Small:Exec
    Consumable_Large:Exec
    Balloon:Exec
    MountedTurret:Exec
    CreativeTool:Exec
    ExplosiveBow:Exec
    AshtonIndigo:Exec
    AshtonChicago:Exec
    MeleeDualWield:Exec
    Unarmed:Exec
   )
   Set using Pistol(
    execute=[
     @n6.AR_BullPup
     @n6.AR_DrumGun
     @n6.AR_ForwardGrip
     @n6.AbilityDecoTool
     @n6.AshtonChicago
     @n6.AshtonIndigo
     @n6.AssaultRifle
     @n6.Balloon
     @n6.C4
     @n6.Consumable_Large
     @n6.Consumable_Small
     @n6.CreativeTool
     @n6.Crossbow
     @n6.DualWield
     @n6.ExplosiveBow
     @n6.GoingCommando
     @n6.GrenadeLauncher
     @n6.MachinePistol
     @n6.MedPackPaddles
     @n6.Melee
     @n6.MeleeDualWield
     @n6.MeleeOneHand
     @n6.MountedTurret
     @n6.PaperBlueprint
     @n6.RemoteControl
     @n6.Rifle
     @n6.RocketLauncher
     @n6.SMG_P90
     @n6.Shotgun
     @n6.ShoulderLauncher
     @n6.SniperRifle
     @n6.TacticalShotgun
     @n6.TrapPlacement
    ]
    using Pistol:double=0
   ) -> (Output_Get:double)
   Set using Pistol(execute=@n6.Pistol, using Pistol=1)
   n7 = GetOwnerSkeletalMeshComponent()
   n8 = Branch(
    execute=@n1.then_3
    Condition="float < float"(
     A=GetAlphaFromReferenceTransformDelta(
      SkeletalMeshComponent=@n7
      BoneName=upperarm_r
      BaseName=clavicle_r
      DeltaComponentType=EulerX
      InputRange=(X=7,Y=-150)
      OutputRange=(X=1,Y=0)
      TwistAxis="1,0,0"
     )
     B=1
    )
   )
   > Set ArmRtDn(ArmRtDn:float=0) -> (Output_Get:float)
   n9 = Get bIsSprinting()
   n10 = Get bIsSkyDiving()
   n11 = float * float!(
    A=Get using Pistol() -> (using Pistol:double)
    B="To Float (Boolean)"(InBool=NOT Boolean(A=@n9))
    C:double="To Float (Boolean)"(InBool=NOT Boolean(A=@n10))
   ) -> (ReturnValue) {is=title K2Node_PromotableOperator}
   Set ArmRtDn(execute=@n8.else, ArmRtDn=@n11)
   K2Node_Event_1 = CustomEvent() {is=title K2Node_CustomEvent}
   > Set Is Sky Diving(Is Sky Diving:double=0) -> (Output_Get:double)
   > n12 = Switch on EGliderType(Selection=Get PlayerGliderType() -> (PlayerGliderType=HangGlider))
   Set IsHangGlider(execute=[@n12.Cape, @n12.NoGlider, @n12.Stairs, @n12.Surfing, @n12.SurfingSimple, @n12.Umbrella], IsHangGlider:double=0) -> (Output_Get:double)
   Delay(execute=@n12.HangGlider, Duration=1)
   > Set IsHangGlider(IsHangGlider=1)
   K2Node_Event_4 = CustomEvent_1() {is=title K2Node_CustomEvent}
   > Set IsHangGlider(IsHangGlider=0)
   n13 = GetOwnerSkeletalMeshComponent()
   n14 = Branch(Condition="float < float"(
     A=GetAlphaFromReferenceTransformDelta(
      SkeletalMeshComponent=@n13
      BoneName=upperarm_l
      BaseName=clavicle_l
      DeltaComponentType=EulerX
      InputRange=(X=0,Y=-100)
      OutputRange=(X=1,Y=0)
      TwistAxis="1,0,0"
     )
     B=1
    ))
   > Set ArmLfDn(ArmLfDn:double=0) -> (Output_Get:double)
   Set ArmLfDn(execute=@n14.else, ArmLfDn=@n11)
   TryGetPawnOwner()
   GetOwnerSkeletalMeshComponent()
   GetOwnerSkeletalMeshComponent()
   GetOwnerSkeletalMeshComponent()
  ]}
]
variables [
 ar_firingPosition {type=number}
 ArmLfDn {type=number}
 ArmRtDn {type=number}
 Gravity Override Params {type=GravityOverrideParamsStruct}
 GravityOverride {type=Vector}
 Is Crouching NotMoving Float {type=number}
 Is Sky Diving {type=number}
 IsHangGlider {type=number}
 using Pistol {type=number}
]
